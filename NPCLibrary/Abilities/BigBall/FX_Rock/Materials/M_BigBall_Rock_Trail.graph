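# Translucent unlit Niagara ribbon material for the smoky trail of the NPC BigBall rock ability effect.
# A panning dissolve texture distorts the UVs of scrolling smoke textures, combined with an oversaturated chromatic layer and tinted by particle color for emissive.
# Smoke luminance and particle alpha feed a depth fade that also drives a subtle index-of-refraction lerp.
# Dynamic parameters set refraction strength, random time offset, scroll speed and desaturation.
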Material /NPCLibrary/Abilities/BigBall/FX_Rock/Materials/M_BigBall_Rock_Trail
graph {
 properties {
  bUsedWithNiagaraRibbons=true
  blendMode=BLEND_Translucent
  pixelDepthOffsetMode=PDOM_Legacy
  refractionMethod=RM_IndexOfRefraction
  shadingModel=MSM_Unlit
  usedShadingModels=Unlit
 }
 nodes [
  n1 = TextureCoordinate() {coordinateIndex=1}
  n2 = DynamicParameter() {defaultValue=<rgba 1 1 1 0>, paramNames [Ref,Random,Speed,Desaturation]}
  n3 = Add(A=Multiply(A=Time(), B=@n2.Speed) {constB=4.6216}, B=Divide(A=@n2.Random) {constB=1.2395})
  n4 = Multiply(A=TextureSample(UVs=Panner(Coordinate=@n1, Time=@n3)) {texture=/Game/CharClasses/Outlander/Gadgets/AntiMaterialCharge/Materials/T_AntiMaterialCharge_Dissolve_01}.R) {constB=0.05}
  n5 = Panner(Coordinate=@n1, Time=@n3) {speedX=-0.4}
  n6 = Multiply(A=TextureSample(UVs=Add(A=@n4, B=@n5)) {texture=/Game/Effects/Fort_Effects/Textures/T_InkySmokeTrail}.G) {constB=3.9124}
  n7 = Add(A=@n4, B=Panner(Coordinate=@n1, Time=@n3) {speedX=-0.3})
  n8 = TextureSample(UVs=@n7) {samplerType=SAMPLERTYPE_Grayscale, texture=/Game/Effects/Fort_Effects/Textures/Smoke/T_Smoke_Trail_Soft_NoClamp}
  n9 = ParticleColor()
  n10 = Multiply(A=Add(A=@n6, B=Multiply(A=@n8.R) {constB=2}), B=@n9.A)
  n11 = DynamicParameter() {defaultValue=<rgba 1 1 1 0>, paramNames [Ref,Random,Speed,Desaturation]}
  LinearInterpolate(Alpha=Multiply(A=DepthFade(Opacity=@n10), B=@n11.Ref)) {constA=1, constB=1.05}
  n12 = Multiply(A=Power(Base=@n8.R) {constExponent=12}) {constB=5}
  n13 = Multiply(A=@n8.R) {constB=2}
  n14 = TextureSample(UVs=@n7) {texture=/Game/Effects/Fort_Effects/Textures/Smoke/T_Smoke_Trail_Soft_Chromatic}
  n15 = Constant() {r=-1}
  n16 = Add(A=@n6, B=Multiply(A=Add(A=@n12, B=@n13), B=Desaturation(None=@n14.RGB, Fraction=@n15)))
  MaterialFunctionCall(Emissive=Multiply(A=Desaturation(None=@n16, Fraction=@n11.Desaturation), B=@n9.RGB)) {materialFunction=&MF_FX_EmissiveAdjustments}
 ]
}
